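# Character-shader material for the Madame Moth backpack, adding an animated two-color emissive glow and a crystal effect between the shared PreFX and PostFX parent functions.
# Panned smoke noise, weighted by a backpack texture raised to a sine-pulsed power, blends the two emissive colors.
# The Tech Art mask's blue channel confines the glow and blackens base color there, while green feeds MF_Crystal's alpha.
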Material /BRCosmetics/Accessories/FORT_Backpacks/Backpack_F_MED_MadameMoth/Materials/M_MadameMoth_Backpack
graph {
 properties {
  bUseMaterialAttributes=true
  bUsedWithMeshDeformer=true
  bUsedWithNanite=true
  bUsedWithSkeletalMesh=true
  pixelDepthOffsetMode=PDOM_Legacy
  shadingModel=MSM_FromMaterialExpression
  usedShadingModels=Default Lit | Subsurface | Clear Coat | Subsurface Profile | Cloth
 }
 nodes [
  n1 = Multiply(A=TextureCoordinate(), B=ScalarParameter() {defaultValue=1, group=MadameMoth, parameterName=Tilling, sortPriority=44})
  n2 = Multiply(A=Time(), B=ScalarParameter() {defaultValue=0.5, group=MadameMoth, parameterName=EmissiveSpeedMult})
  n3 = TextureObjectParameter() {
   group=MadameMoth
   parameterName=TexNoise
   sortPriority=0
   texture=/Game/Global/Textures/Default/Tileable_Patterns/T_Blowing_Smoke_Pattern2_Tile_DISP
  }
  n4 = TextureCoordinate() {uTiling=0.7, vTiling=0.8}
  n5 = ScalarParameter() {defaultValue=1, group=MadameMoth, parameterName=Tilling, sortPriority=44}
  n6 = Time()
  n7 = ScalarParameter() {defaultValue=0.5, group=MadameMoth, parameterName=EmissiveSpeedMult}
  n8 = TextureSample(UVs=Panner(Coordinate=Multiply(A=@n4, B=@n5), Time=Multiply(A=@n6, B=@n7)) {speedX=-0.005, speedY=0.01}, Tex=@n3)
  Add(A=TextureSample(UVs=Panner(Coordinate=@n1, Time=@n2) {speedX=-0.01}, Tex=@n3).R, B=@n8.R)
  n9 = MaterialFunctionCall() {materialFunction=&MF_CharacterParent_2023_PreFX}
  n10 = GetMaterialAttributes(None=@n9.Result) {attributeGetTypes [#1, #2]}
  n11 = ComponentMask(None=@n9.Tech Art Mask) {b=true}
  n12 = VectorParameter() {group=MadameMoth, parameterName=EmissiveColorB, defaultValue=<rgba 7 0.0006 0.0329 1>}
  n13 = VectorParameter() {group=MadameMoth, parameterName=EmissiveColor, defaultValue=<rgba 7 0.0006 0.0329 1>}
  n14 = TextureSample() {
   samplerType=SAMPLERTYPE_LinearColor
   texture=/BRCosmetics/Accessories/FORT_Backpacks/Backpack_F_MED_MadameMoth/Textures/T_MadameMoth_Backpack_TA_E
  }
  n15 = ScalarParameter() {defaultValue=1, group=MadameMoth, parameterName=FlamesNoisePow, sortPriority=33}
  n16 = Time()
  n17 = Multiply(A=Sine(None=Multiply(A=@n16) {constB=0.2})) {constB=0.3}
  n18 = LinearInterpolate(B=Divide(A=@n8.R), Alpha=Power(Base=@n14.R, Exp=Add(A=@n15, B=@n17)))
  n19 = Multiply(
   A=Multiply(A=@n11, B=LinearInterpolate(A=@n12.RGB, B=@n13.RGB, Alpha=@n18)) {constA=1}
   B=ScalarParameter() {defaultValue=2, group=MadameMoth, parameterName=EmissiveIntensity, sortPriority=20}
  )
  n20 = LinearInterpolate(A=@n10.Base Color, Alpha=@n11) {constB=0}
  MaterialFunctionCall(
   Material=MaterialFunctionCall!(
    Material=SetMaterialAttributes!(MaterialAttributes=@n10.MaterialAttributes, Emissive Color=@n19, Base Color=@n20) {attributeSetTypes [#1, #2]}
    Alpha=ComponentMask(None=@n9.Tech Art Mask) {g=true}
   ) {materialFunction=/Game/Global/MaterialFunctions/TechArt/MF_Crystal}.Result
   M=@n9.M
   Pre Skinned Local Position=@n9.M
   Distance Blend=@n9.M
   Fresnel=@n9.M
  ) {materialFunction=&MF_CharacterParent_2023_PostFX}
  SmoothStep(Min=ScalarParameter() {defaultValue=0.3, group=MadameMoth, parameterName=FlamesNoiseMin}, Max=@n15, Value=@n14.R) {constMin=0.3}
  StaticComponentMaskParameter(None=@n10.Emissive Color) {defaultR=true, group=MadameMoth, parameterName=MaskColor, sortPriority=40}
 ]
}
